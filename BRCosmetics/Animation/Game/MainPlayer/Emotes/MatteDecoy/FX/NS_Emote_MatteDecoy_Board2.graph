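# Niagara system for the MatteDecoy emote board effect: a glowing checkerboard of tiles that treadmills around the player, colored by particle partition parity.
# The Chessboard emitter traces tiles onto surfaces for sprites and decals, with distance-based fades and waves plus indoor dimming; its grid is empty in the front end.
# FE_Train keeps flat, upward-facing tiles at socket height, scrolling them in the front end.
# System Rail Float and its inverse scale the two emitters oppositely.
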NiagaraSystem /BRCosmetics/Animation/Game/MainPlayer/Emotes/MatteDecoy/FX/NS_Emote_MatteDecoy_Board2
dependencies {
 usedDataInterfaces [
  &NiagaraDataInterfaceCurve
  &NiagaraDataInterfaceMeshRendererInfo
  &NiagaraDataInterfaceArrayFloat
  &NiagaraDataInterfaceCollisionQuery
  &NiagaraDataInterfaceDebugDraw
  &NiagaraDataInterfaceSkeletalMesh
  &NiagaraDataInterfaceVector2DCurve
  &NiagaraDataInterfaceSocketReader
  &NiagaraDataInterfaceArrayFloat3
 ]
 usedDynamicInputs [&SelectVectorFromArray, /Niagara/DynamicInputs/Lerp/SmoothLerpOverTime_Float]
 usedModules [
  &SystemState
  &BlendSpaceHelper
  &self:SystemSpawnScript.SystemScriptSource.SystemScriptGraph.NiagaraNodeAssignment_6.SetVariables_#1
  &self:SystemSpawnScript.SystemScriptSource.SystemScriptGraph.NiagaraNodeAssignment_1.SetVariables_#2
  &self:SystemSpawnScript.SystemScriptSource.SystemScriptGraph.NiagaraNodeAssignment_0.SetVariables_#3
  &self:SystemSpawnScript.SystemScriptSource.SystemScriptGraph.NiagaraNodeAssignment_2.SetVariables_#4
 ]
}
emitters {
 Chessboard {
  emitterUpdateScript [
   &EmitterState(
    Inactive Response=::NewEnumerator1
    Life Cycle Mode=::NewEnumerator1
    Loop Behavior=::NewEnumerator1
    Loop Duration Mode=::NewEnumerator1
    UseLoopDelay=false
   )
   /SetVariables(
    Emitter.OriginOffset:Vector3f=&TransformVector(Vector=<0 0 0>, Destination Space=::World)
    Emitter.PaddingPerCell=2
    Emitter.VolumeRadius=200
   ) {isSetParameters=true, module=SetVariables_#5}
   /Niagara/Modules/Spawn/Location/SpawnParticlesInGrid(
    X Count:NiagaraInt32=&MakeIntFromFloat(Float=&FN_FrontEndFloatModifier(Gameplay=4, Front End=0))
    Y Count:NiagaraInt32=&MakeIntFromFloat(Float=&FN_FrontEndFloatModifier(Gameplay=4, Front End=0))
    Z Count=0
   )
   &self:ScratchModule(DistanceMultiplier=1.1, Trace Origin Height=150, TraceDir:Vector3f=<0 0 -1>, TraceDist=300) {module=ScratchModule}
  ]
  particleSpawnScript [
   &FN_SetSpawnGroup() {module=FN_SetSpawnGroup001}
   &InitializeParticle(
    Position Offset=<-8 0 0>
    Position Offset Coordinate Space=::Local
    Sprite Rotation Mode=::NewEnumerator0
    Sprite Size Mode=::NewEnumerator3
    Uniform Sprite Size=111
    UsePositionOffset=true
    Write Material Random=true
   )
   &SocketLocation(
    Orientation Sampling=::NewEnumerator1
    Position Sampling=::NewEnumerator1
    SamplePosOffset=false
    Socket Reader=$System.Socket Reader
    Socket Sampling Mode=::NewEnumerator3
    VelocitySpeedLimit=false
   )
   /Niagara/Modules/Spawn/Location/V2/GridLocation(
    Coordinate Space:ENiagaraCoordinateSpace=::World
    Module.Jitter Inside Grid Cells=false
    Module.Override Random Seed=false
    Normalize Offsets=false
    XYZ Dimensions:Vector3f=&MakeVector(
     X=&Divide_Float(A=$Emitter.VolumeRadius, B=$Emitter.PaddingPerCell)
     Y=&Divide_Float(A=$Emitter.VolumeRadius, B=$Emitter.PaddingPerCell)
     Z=0
    )
   )
   /SetVariables() {isSetParameters=true}
   &PartitionParticles(Number Of Partitions=8)
   &CameraOffset() {module=CameraOffset001}
   /SetVariables(Particles.LOGIC COLOR:LinearColor=&MakeCustomLinearColorFromBool(
     Bool=&SetBoolByFloatComparison(
      A=&MakeFloatFromInt(Int=$Particles.PartitionParticles.Partition)
      B=4
      Comparison Type=::NewEnumerator2
     )
     True Color=&MakeCustomLinearColorFromBool(
      Bool=&SetBoolByFloatComparison(
       A=&MakeFloatFromInt(Int=&ModuloInt(A=$Particles.PartitionParticles.Partition, B=2))
       B=0
      )
      True Color=<rgba 0.4219 0.533 1 1>
      False Color=<rgba 0.0503 0.0544 0.0938 1>
     )
     False Color=&MakeCustomLinearColorFromBool(
      Bool=&SetBoolByFloatComparison(
       A=&MakeFloatFromInt(Int=&ModuloInt(A=$Particles.PartitionParticles.Partition, B=2))
       B=0
      )
      True Color=<rgba 0.0503 0.0544 0.0938 1>
      False Color=<rgba 0.4219 0.533 1 1>
     )
    )) {isSetParameters=true}
   /SetVariables(Particles.DARKCOLOR:LinearColor=<rgba 8.593 15.6344 40 1>) {isSetParameters=true}
   /SetVariables(Particles.DecalSize:Vector3f=<300 110 110>) {isSetParameters=true}
   /SetVariables(Particles.DecalGlowSize:Vector3f=<300 230 230>) {isSetParameters=true}
   /SetVariables(Particles.Decal Base Color:LinearColor=&MakeCustomLinearColorFromBool(
     Bool=&SetBoolByFloatComparison(
      A=&MakeFloatFromInt(Int=$Particles.PartitionParticles.Partition)
      B=4
      Comparison Type=::NewEnumerator2
     )
     True Color=&MakeCustomLinearColorFromBool(
      Bool=&SetBoolByFloatComparison(
       A=&MakeFloatFromInt(Int=&ModuloInt(A=$Particles.PartitionParticles.Partition, B=2))
       B=0
      )
      True Color=<rgba 0.1495 0.156 0.2031 1>
      False Color=<rgba 0.002 0.0033 0.0156 1>
     )
     False Color=&MakeCustomLinearColorFromBool(
      Bool=&SetBoolByFloatComparison(
       A=&MakeFloatFromInt(Int=&ModuloInt(A=$Particles.PartitionParticles.Partition, B=2))
       B=0
      )
      True Color=<rgba 0.002 0.0033 0.0156 1>
      False Color=<rgba 0.1471 0.1541 0.2031 1>
     )
    )) {isSetParameters=true}
   /SetVariables(Particles.Decal Glow Color:LinearColor=$Particles.LOGIC COLOR) {isSetParameters=true}
   /SetVariables(Particles.Decal Edge Color:LinearColor=$Particles.DARKCOLOR) {isSetParameters=true}
   /SetVariables(
    Particles.DecalGlowSize:Vector3f=&Multiply_VectorByFloat(Vector=$Particles.DecalGlowSize, Float=0.91)
    Particles.DecalSize:Vector3f=&Multiply_VectorByFloat(Vector=$Particles.DecalSize, Float=0.91)
    Particles.SpriteSize=&Multiply_Vector2D_ByFloat(Float=0.91, Vector2D=$Particles.SpriteSize)
   ) {isSetParameters=true}
   /SetVariables(
    Particles.Decal Base Color:LinearColor=&Multiply_LinearColorByFloat(
     Linear Color=$Particles.Decal Base Color
     Float=&FN_FrontEndFloatModifier(Gameplay=1, Front End=0)
    )
    Particles.Decal Edge Color:LinearColor=&Multiply_LinearColorByFloat(
     Linear Color=$Particles.Decal Edge Color
     Float=&FN_FrontEndFloatModifier(Gameplay=1, Front End=0)
    )
    Particles.Decal Glow Color:LinearColor=&Multiply_LinearColorByFloat(
     Linear Color=$Particles.Decal Glow Color
     Float=&FN_FrontEndFloatModifier(Gameplay=1, Front End=0)
    )
   ) {isSetParameters=true}
  ]
  particleUpdateScript [
   &ParticleState(Kill Particles When Lifetime Has Elapsed=false, Let Infinitely Lived Particles Die When Emitter Deactivates=true)
   /SetVariables(Particles.IndoorMod=0.05) {isSetParameters=true}
   &SkeletalMeshLocation(
    Bone Sampling Mode=::NewEnumerator4
    MaskSpawnGroup=false
    Orientation Sampling=::NewEnumerator1
    Position Sampling=::NewEnumerator2
    Skeletal Mesh=DataInterface() {propertyValues {
      DefaultMesh=null
      ExcludeBoneName=root
      OptionalComponentClass=null
      PreviewMesh=null
      SkinningMode=SkinOnTheFly
      SoftSourceActor=null
      SourceMode=Default
      UvSetIndex=0
      WholeMeshLOD=-1
      bExcludeBone=false
      bReadDeformedGeometry=true
      bRequireCurrentFrameData=true
      ComponentTags []
      FilteredBones [root]
      FilteredSockets []
      MeshUserParameter {parameter {name=None, typeDefHandle {registeredTypeIndex=1}, varData []}}
      SamplingRegions []
     }}
   )
   &CurlNoiseForce(Bias Noise=false, Noise Strength=150, Pan Noise=true, Pan Noise Field=<1 1 1>, Write to Intrinsic Parameters=false)
   &SolveForcesAndVelocity(Clamp Velocity=false, Limit Acceleration=false)
   /SetVariables(Particles.PositionCenter:NiagaraPosition=&AddVectorToPosition(Position=$Engine.Owner.Position, Vector=$Emitter.OriginOffset)) {isSetParameters=true, module=SetVariables_#6}
   &self:TreadmillParticles(
    GridSize:NiagaraFloat=&Multiply_Float(A=$Emitter.VolumeRadius, B=2)
    InputPosition:NiagaraPosition=$Particles.PositionCenter
   ) {module=TreadmillParticles}
   &DoOnce(
    Reset:NiagaraBool=&SetBoolByFloatComparison(A=$Particles.Timeline001.Time, B=0.3)
    Trigger Condition:NiagaraBool=&SetBoolByFloatComparison(
     A=&VectorLength(Vector=Linked() {linkedVariable {name=Particles.Position, type=NiagaraPosition}})
     B=&VectorLength(Vector=Linked() {linkedVariable {name=Particles.Previous.Position, type=NiagaraPosition}})
     Comparison Type=::NewEnumerator5
    )
   )
   &Timeline(Instantaneous Rewind=true, Loop While Playing=true, Play=$Particles.DoOnce.TriggeredOnce, Play Rate=10) {module=Timeline001}
   &self:GridCollision_02(
    CharacterPos:NiagaraPosition=&AddVectorToPosition(Position=$Engine.Owner.Position, Vector=<0 0 50>)
    Disable Trace:NiagaraBool=&Boolean_AndOperation(
     Bool1=&Boolean_NotOperation(Bool=$Particles.DoOnce.TriggeredOnce)
     Bool2=&SetBoolByFloatComparison(A=$Emitter.Age, B=0.4)
    )
    Height Adjustment=2.2
    Trace Origin Height:NiagaraFloat=&Add_Float(
     A=&self:ScratchDynamicInput_09(
      ForwardVector:Vector3f=&ReturnMeshOrientationAxis(Mesh Orientation Axis=::YAxis, Mesh Orientation Quat=$Particles.SkeletalMeshLocation.SampledRotation)
      FinalOffset:NiagaraFloat=150
     )
     B=0
    )
    TraceDir:Vector3f=<0 0 -1>
    TraceDist=350
    bIsFrontEnd:NiagaraBool=$User.bIsFrontEndPreview
   ) {module=GridCollision_02}
   /SetVariables(Particles.Position=$Particles.CollisionPos) {isSetParameters=true}
   /SetVariables(Particles.RemappedDistance:NiagaraFloat=&NormalizeDistanceRange(
     Start Position:NiagaraPosition=/Niagara/DynamicInputs/Helpers/SimulationPosition()
     End Position:NiagaraPosition=$Particles.Position
     Distance:NiagaraFloat=&Multiply_Float(A=340, B=0.6)
     Invert Normalized Range:NiagaraBool=false
    )) {isSetParameters=true}
   /SetVariables(
    Particles.BaseDirection:Vector3f=$Particles.CollisionNormal
    Particles.WaveFormDirection:Vector3f=&Multiply_Vector(A=$Particles.CollisionNormal, B=<1 1 1>)
   ) {isSetParameters=true}
   /SetVariables(Particles.SpriteFacing=$Particles.CollisionNormal) {isSetParameters=true}
   /SetVariables() {isSetParameters=true}
   /SetVariables(Particles.NEWPOSITION:NiagaraPosition=&AddVectorToPosition(
     Position=&AddVectorToPosition(
      Position=$Particles.Position
      Vector=&Add_Vector(
       A=&Multiply_Vector(
        A=$Particles.CollisionNormal
        B=&VectorFromFloat(Value=&MakeFloatFromVector(
          Vector=&VectorFromFloat(Value=&Waveform(
            Global Amplitude Scale=&Multiply_Float(
             A=&FloatFromCurve(FloatCurve=DataInterface() {propertyValues {Curve {keys [<at 0.5 0>,<at 1 1>]}}}, CurveIndex=$Particles.RemappedDistance)
             B=&RandomRangeFloat(Minimum=10, Maximum=15)
            )
            "Waveform [1]"=::NewEnumerator0
            "[1] Frequency"=&RandomRangeFloat(Minimum=0.1, Maximum=0.2)
            "[1] Amplitude Min/Max"=<0.1 1>
           ))
          Channel=::NewEnumerator0
         ))
       )
       B=&Lerp_Vector(
        A=&Multiply_Vector(A=$Particles.CollisionNormal, B=&VectorFromFloat(Value=30))
        B=&Multiply_Vector(A=$Particles.CollisionNormal, B=<2 2 2>)
        Alpha=&DotProduct(Vector A=<0 0 1>, Vector B=$Particles.SpriteFacing)
       )
      )
     )
     Vector=&Lerp_Vector(
      A=&Multiply_Vector(
       A=$Particles.CollisionNormal
       B=&VectorFromFloat(Value=&FloatFromCurve(
         FloatCurve=DataInterface() {propertyValues {Curve {keys [
             {arriveTangentWeight=0.1562, leaveTangentWeight=0.2983, time=0.4, value=0}
             {arriveTangent=5, time=0.89, value=1}
             <at 0.95 1>
             {interpMode=RCIM_Cubic, tangentMode=RCTM_Break, time=1, value=0}
            ]}}}
         CurveIndex=$Particles.RemappedDistance
         Scale Curve=5
        ))
      )
      B=&Multiply_Vector(
       A=$Particles.CollisionNormal
       B=&VectorFromFloat(Value=&FloatFromCurve(
         FloatCurve=DataInterface() {propertyValues {Curve {keys [
             {arriveTangentWeight=0.1562, leaveTangentWeight=0.2983, time=0.4, value=0}
             <at 0.5002 1>
             {arriveTangent=5, time=0.8, value=1}
             {interpMode=RCIM_Cubic, tangentMode=RCTM_Break, time=1, value=0}
            ]}}}
         CurveIndex=$Particles.RemappedDistance
         Scale Curve=0
        ))
      )
      Alpha=&DotProduct(Vector A=$Particles.CollisionNormal, Vector B=<0 0 1>)
     )
    )) {isSetParameters=true}
   /SetVariables(Particles.NEWFACING:Vector3f=&Add_Vector(
     A=$Particles.CollisionNormal
     B=&MakeVector(
      X=&Waveform(
       Global Amplitude Scale=&Multiply_Float(
        A=&FloatFromCurve(FloatCurve=DataInterface() {propertyValues {Curve {keys [<at 0.41 0>,<at 0.6 1>]}}}, CurveIndex=$Particles.RemappedDistance)
        B=&RandomRangeFloat(Minimum=0.02, Maximum=0.03)
       )
       "Waveform [1]"=::NewEnumerator0
       "[1] Frequency"=&RandomRangeFloat(Minimum=0.25, Maximum=0.5)
       "[1] Amplitude Min/Max"=<-1 1>
      )
      Y=&Waveform(
       Global Amplitude Scale=&Multiply_Float(A=$Particles.RemappedDistance, B=0.05)
       "Waveform [1]"=::NewEnumerator0
       "[1] Frequency"=&RandomRangeFloat(Minimum=0.25, Maximum=0.5)
       "[1] Amplitude Min/Max"=<-1 1>
      )
      Z=0
     )
    )) {isSetParameters=true}
   /SetVariables(Particles.MULTIPLIED COLOR:LinearColor=&Multiply_LinearColorByFloat(
     Linear Color=&MultiplyLinearColor(
      Linear Color=$Particles.LOGIC COLOR
      Scale Factor=&MakeVector4(
       X=&FloatFromCurve(
        FloatCurve=DataInterface() {propertyValues {Curve {
           defaultValue=1
           keys [{interpMode=RCIM_Cubic, tangentMode=RCTM_User, time=0, value=0}, {interpMode=RCIM_Cubic, tangentMode=RCTM_User, time=1, value=1}]
          }}}
        Scale Curve=&Lerp_Float(A=1, B=0.5, Alpha=$Particles.INTERPOLATE INDOORS)
       )
       Y=&FloatFromCurve(
        FloatCurve=DataInterface() {propertyValues {Curve {
           defaultValue=1
           keys [{interpMode=RCIM_Cubic, tangentMode=RCTM_User, time=0, value=0}, {interpMode=RCIM_Cubic, tangentMode=RCTM_User, time=1, value=1}]
          }}}
        Scale Curve=&Lerp_Float(A=1, B=0.5, Alpha=$Particles.INTERPOLATE INDOORS)
       )
       Z=&FloatFromCurve(
        FloatCurve=DataInterface() {propertyValues {Curve {
           defaultValue=1
           keys [{interpMode=RCIM_Cubic, tangentMode=RCTM_User, time=0, value=0}, {interpMode=RCIM_Cubic, tangentMode=RCTM_User, time=1, value=1}]
          }}}
        Scale Curve=&Lerp_Float(A=1, B=0.5, Alpha=$Particles.INTERPOLATE INDOORS)
       )
       W=&Multiply_Float(
        A=&FloatFromCurve(
         FloatCurve=DataInterface() {propertyValues {Curve {keys [
             {arriveTangentWeight=0.1562, leaveTangentWeight=0.2983, time=0.4, value=0}
             {arriveTangent=5, time=0.401, value=0.5}
             {interpMode=RCIM_Cubic, tangentMode=RCTM_Break, time=0.41, value=0.5}
             {arriveTangent=-0.8475, interpMode=RCIM_Cubic, leaveTangent=-0.8475, time=0.45, value=0.35}
             {arriveTangentWeight=0.0898, interpMode=RCIM_Cubic, leaveTangentWeight=0.2693, tangentMode=RCTM_User, time=1, value=0}
            ]}}}
         CurveIndex=$Particles.RemappedDistance
         Scale Curve=$System.Rail Float Inverted
        )
        B=&FloatFromCurve(FloatCurve=DataInterface() {propertyValues {Curve {
            defaultValue=1
            keys [{interpMode=RCIM_Cubic, tangentMode=RCTM_User, time=0, value=0}, {interpMode=RCIM_Cubic, tangentMode=RCTM_User, time=1, value=1}]
           }}})
       )
      )
     )
     Float=&MakeCustomFloatFromBool(Bool=true, True Float=1, False Float=0)
    )) {isSetParameters=true}
   /SetVariables(Particles.MULTIPLIED DARk:LinearColor=&MultiplyLinearColor(
     Linear Color=$Particles.DARKCOLOR
     Scale Factor=&MakeVector4(
      X=&FloatFromCurve(
       FloatCurve=DataInterface() {propertyValues {Curve {
          defaultValue=1
          keys [{interpMode=RCIM_Cubic, tangentMode=RCTM_User, time=0, value=0}, {interpMode=RCIM_Cubic, tangentMode=RCTM_User, time=1, value=1}]
         }}}
       Scale Curve=&Lerp_Float(A=1, B=0.18, Alpha=$Particles.INTERPOLATE INDOORS)
      )
      Y=&FloatFromCurve(
       FloatCurve=DataInterface() {propertyValues {Curve {
          defaultValue=1
          keys [{interpMode=RCIM_Cubic, tangentMode=RCTM_User, time=0, value=0}, {interpMode=RCIM_Cubic, tangentMode=RCTM_User, time=1, value=1}]
         }}}
       Scale Curve=&Lerp_Float(A=1, B=0.18, Alpha=$Particles.INTERPOLATE INDOORS)
      )
      Z=&FloatFromCurve(
       FloatCurve=DataInterface() {propertyValues {Curve {
          defaultValue=1
          keys [{interpMode=RCIM_Cubic, tangentMode=RCTM_User, time=0, value=0}, {interpMode=RCIM_Cubic, tangentMode=RCTM_User, time=1, value=1}]
         }}}
       Scale Curve=&Lerp_Float(A=1, B=0.18, Alpha=$Particles.INTERPOLATE INDOORS)
      )
      W=&Multiply_Float(
       A=&FloatFromCurve(
        FloatCurve=DataInterface() {propertyValues {Curve {keys [
            {arriveTangentWeight=0.1562, leaveTangentWeight=0.2983, time=0.4, value=0}
            {arriveTangent=5, time=0.401, value=1}
            {arriveTangentWeight=0.0898, interpMode=RCIM_Cubic, leaveTangentWeight=0.2693, tangentMode=RCTM_User, time=0.42, value=1}
            {interpMode=RCIM_Cubic, tangentMode=RCTM_Break, time=1, value=0}
           ]}}}
        CurveIndex=$Particles.RemappedDistance
        Scale Curve=$System.Rail Float Inverted
       )
       B=&FloatFromCurve(FloatCurve=DataInterface() {propertyValues {Curve {
           defaultValue=1
           keys [{interpMode=RCIM_Cubic, tangentMode=RCTM_User, time=0, value=0}, {interpMode=RCIM_Cubic, tangentMode=RCTM_User, time=1, value=1}]
          }}})
      )
     )
    )) {isSetParameters=true}
   &RegionMask(
    Invert Region=true
    Offset:Vector3f=$Emitter.OriginOffset
    Region Dimensions:Vector3f=<2500 2500 2500>
    Region Falloff=0
    Region Type:ENiagaraDebugDrawShape=::NewEnumerator1
    UseOffset=true
    UseRotation=false
   )
   &Timeline(
    Additional Output:ENiagara_TimelineOutput=::NewEnumerator1
    Additional Output Mode:ENiagara_TimelineOutputMode=::NewEnumerator1
    Float Curve=DataInterface() {propertyValues {Curve {keys [{arriveTangent=0.9218, interpMode=RCIM_Cubic, leaveTangent=0.9218, tangentMode=RCTM_User, time=0, value=0}, <at 1 1 RCIM_Cubic>]}}}
    Play=&Boolean_NotOperation(Bool=$Particles.RegionMask.IsInsideMask)
    Play Rate=&MakeCustomFloatFromBool(
     Bool=&SetBoolByFloatComparison(A=&VectorLength(Vector=$User.OwnerLocalVelocity), B=160)
     True Float=10
     False Float=1
    )
    Rewind Rate:NiagaraFloat=&MakeCustomFloatFromBool(
     Bool=&SetBoolByFloatComparison(A=&VectorLength(Vector=$User.OwnerLocalVelocity), B=160)
     True Float=10
     False Float=3
    )
   )
   &Color(
    Color=&MakeCustomLinearColorFromBool(
     Bool=&SetBoolByFloatComparison(
      A=&MakeFloatFromInt(Int=$Particles.PartitionParticles.Partition)
      B=4
      Comparison Type=::NewEnumerator2
     )
     True Color=&MakeCustomLinearColorFromBool(
      Bool=&SetBoolByFloatComparison(
       A=&MakeFloatFromInt(Int=&ModuloInt(A=$Particles.PartitionParticles.Partition, B=2))
       B=0
      )
      True Color=<rgba 0.1495 0.156 0.2031 1>
      False Color=<rgba 0.002 0.0033 0.0156 1>
     )
     False Color=&MakeCustomLinearColorFromBool(
      Bool=&SetBoolByFloatComparison(
       A=&MakeFloatFromInt(Int=&ModuloInt(A=$Particles.PartitionParticles.Partition, B=2))
       B=0
      )
      True Color=<rgba 0.002 0.0033 0.0156 1>
      False Color=<rgba 0.1471 0.1541 0.2031 1>
     )
    )
    Scale Alpha=&Multiply_Float(
     A=&FloatFromCurve(
      FloatCurve=DataInterface() {propertyValues {Curve {keys [
          {arriveTangentWeight=0.1562, leaveTangentWeight=0.2983, time=0.4, value=0}
          {arriveTangent=5, time=0.401, value=1}
          {arriveTangentWeight=0.0898, interpMode=RCIM_Cubic, leaveTangentWeight=0.2693, tangentMode=RCTM_User, time=0.5, value=1}
          {interpMode=RCIM_Cubic, tangentMode=RCTM_Break, time=1, value=0}
         ]}}}
      CurveIndex=$Particles.RemappedDistance
      Scale Curve=$System.Rail Float Inverted
     )
     B=&FloatFromCurve(FloatCurve=DataInterface() {propertyValues {Curve {
         defaultValue=1
         keys [{interpMode=RCIM_Cubic, tangentMode=RCTM_User, time=0, value=0}, {interpMode=RCIM_Cubic, tangentMode=RCTM_User, time=1, value=1}]
        }}})
    )
    Scale Color=&Multiply_VectorByFloat(
     Vector=&VectorFromFloat(Value=&Multiply_Float(A=$Particles.Timeline.Float, B=&Multiply_Float(A=1, B=1)))
     Float=&Lerp_Float(A=1, B=0.1, Alpha=$Particles.INTERPOLATE INDOORS)
    )
   ) {module=Color002}
   /SetVariables(Particles.Decal Base Color:LinearColor=&MultiplyLinearColor(
     Linear Color=$Particles.Initial.Decal Base Color
     Scale Factor=&MakeVector4(
      X=&Lerp_Float(A=1, B=$Particles.IndoorMod, Alpha=$Particles.INTERPOLATE INDOORS)
      Y=&Lerp_Float(A=1, B=$Particles.IndoorMod, Alpha=$Particles.INTERPOLATE INDOORS)
      Z=&Lerp_Float(A=1, B=$Particles.IndoorMod, Alpha=$Particles.INTERPOLATE INDOORS)
      W=&Multiply_Float(
       A=&FloatFromCurve(
        FloatCurve=DataInterface() {propertyValues {Curve {keys [{arriveTangentWeight=0.2522, leaveTangentWeight=0.0925, time=0.401, value=1}, <at 0.402 0>]}}}
        CurveIndex=$Particles.RemappedDistance
        Scale Curve=$System.Rail Float Inverted
       )
       B=&FloatFromCurve(FloatCurve=DataInterface() {propertyValues {Curve {
           defaultValue=1
           keys [{interpMode=RCIM_Cubic, tangentMode=RCTM_User, time=0, value=0}, {interpMode=RCIM_Cubic, tangentMode=RCTM_User, time=1, value=1}]
          }}})
      )
     )
    )) {isSetParameters=true}
   /SetVariables(Particles.Decal Glow Color:LinearColor=&MultiplyLinearColor(
     Linear Color=$Particles.Initial.Decal Glow Color
     Scale Factor=&MakeVector4(
      X=&Lerp_Float(A=1, B=$Particles.IndoorMod, Alpha=$Particles.INTERPOLATE INDOORS)
      Y=&Lerp_Float(A=1, B=$Particles.IndoorMod, Alpha=$Particles.INTERPOLATE INDOORS)
      Z=&Lerp_Float(A=1, B=$Particles.IndoorMod, Alpha=$Particles.INTERPOLATE INDOORS)
      W=&Multiply_Float(
       A=&FloatFromCurve(
        FloatCurve=DataInterface() {propertyValues {Curve {keys [{arriveTangentWeight=0.2522, leaveTangentWeight=0.0925, time=0.401, value=0.5}, <at 0.402 0>]}}}
        CurveIndex=$Particles.RemappedDistance
        Scale Curve=$System.Rail Float Inverted
       )
       B=&FloatFromCurve(FloatCurve=DataInterface() {propertyValues {Curve {
           defaultValue=1
           keys [{interpMode=RCIM_Cubic, tangentMode=RCTM_User, time=0, value=0}, {interpMode=RCIM_Cubic, tangentMode=RCTM_User, time=1, value=1}]
          }}})
      )
     )
    )) {isSetParameters=true}
   /SetVariables(Particles.Decal Edge Color:LinearColor=&MultiplyLinearColor(
     Linear Color=$Particles.Initial.Decal Edge Color
     Scale Factor=&MakeVector4(
      X=&Lerp_Float(A=0.2, B=&Multiply_Float(A=$Particles.IndoorMod, B=10), Alpha=$Particles.INTERPOLATE INDOORS)
      Y=&Lerp_Float(A=0.2, B=&Multiply_Float(A=$Particles.IndoorMod, B=10), Alpha=$Particles.INTERPOLATE INDOORS)
      Z=&Lerp_Float(A=0.2, B=&Multiply_Float(A=$Particles.IndoorMod, B=10), Alpha=$Particles.INTERPOLATE INDOORS)
      W=&Multiply_Float(
       A=&FloatFromCurve(
        FloatCurve=DataInterface() {propertyValues {Curve {keys [{arriveTangentWeight=0.2522, leaveTangentWeight=0.0925, time=0.401, value=0.5}, <at 0.402 0>]}}}
        CurveIndex=$Particles.RemappedDistance
        Scale Curve=$System.Rail Float Inverted
       )
       B=&FloatFromCurve(
        FloatCurve=DataInterface() {propertyValues {Curve {
           defaultValue=1
           keys [{interpMode=RCIM_Cubic, tangentMode=RCTM_User, time=0, value=0}, {interpMode=RCIM_Cubic, tangentMode=RCTM_User, time=1, value=1}]
          }}}
        Scale Curve=4
       )
      )
     )
    )) {isSetParameters=true}
   &ScaleSpriteSize(
    Scale Sprite Size Mode=::NewEnumerator0
    Uniform Scale Factor=&Multiply_Float(
     A=&Multiply_Float(
      A=$System.Rail Float Inverted
      B=&FloatFromCurve(
       FloatCurve=DataInterface() {propertyValues {Curve {keys [<at 0 1>,<at 0.01 1.3>,<at 1 1>]}}}
       CurveIndex=&DotProduct(Vector A=<0 0 1>, Vector B=$Particles.SpriteFacing)
      )
     )
     B=&Multiply_Float(
      A=&FloatFromCurve(
       FloatCurve=DataInterface() {propertyValues {Curve {keys [
           {arriveTangentWeight=0.0898, interpMode=RCIM_Cubic, leaveTangentWeight=0.2693, tangentMode=RCTM_User, time=0.7, value=1}
           {interpMode=RCIM_Cubic, tangentMode=RCTM_Break, time=0.9, value=0.85}
          ]}}}
       CurveIndex=$Particles.RemappedDistance
       Scale Curve=$System.Rail Float Inverted
      )
      B=&FloatFromCurve(FloatCurve=DataInterface() {propertyValues {Curve {
          defaultValue=0
          keys [{interpMode=RCIM_Cubic, tangentMode=RCTM_User, time=0, value=0.9}, {interpMode=RCIM_Cubic, tangentMode=RCTM_User, time=0.6, value=1}]
         }}})
     )
    )
   )
   /SetVariables(Particles.GlowSize=&Multiply_Vector2D_ByFloat(Float=1.5, Vector2D=$Particles.SpriteSize)) {isSetParameters=true, module=SetVariables_#7}
   /SetVariables(
    Particles.DecalGlowSize:Vector3f=&Multiply_VectorByFloat(
     Vector=$Particles.Initial.DecalGlowSize
     Float=&FloatFromCurve(FloatCurve=DataInterface() {propertyValues {Curve {
         defaultValue=1
         keys [{interpMode=RCIM_Cubic, tangentMode=RCTM_User, time=0, value=0.97}, {interpMode=RCIM_Cubic, tangentMode=RCTM_User, time=0.6, value=1}]
        }}})
    )
    Particles.DecalSize:Vector3f=&Multiply_VectorByFloat(
     Vector=$Particles.Initial.DecalSize
     Float=&FloatFromCurve(FloatCurve=DataInterface() {propertyValues {Curve {
         defaultValue=1
         keys [{interpMode=RCIM_Cubic, tangentMode=RCTM_User, time=0, value=0.97}, {interpMode=RCIM_Cubic, tangentMode=RCTM_User, time=0.6, value=1}]
        }}})
    )
   ) {isSetParameters=true}
   /SetVariables(
    Particles.Color=&MakeCustomLinearColorFromBool(
     Bool=&SetBoolByFloatComparison(A=&DotProduct(Vector A=$Particles.SpriteFacing, Vector B=<0 0 1>), B=0, Comparison Type=::NewEnumerator4)
     True Color=<rgba 1 1 1 0>
     False Color=$Particles.Color
    )
    Particles.MULTIPLIED COLOR:LinearColor=&MakeCustomLinearColorFromBool(
     Bool=&SetBoolByFloatComparison(A=&DotProduct(Vector A=$Particles.SpriteFacing, Vector B=<0 0 1>), B=0, Comparison Type=::NewEnumerator4)
     True Color=<rgba 1 1 1 0>
     False Color=$Particles.MULTIPLIED COLOR
    )
    Particles.MULTIPLIED DARk:LinearColor=&MakeCustomLinearColorFromBool(
     Bool=&SetBoolByFloatComparison(A=&DotProduct(Vector A=$Particles.SpriteFacing, Vector B=<0 0 1>), B=0, Comparison Type=::NewEnumerator4)
     True Color=<rgba 1 1 1 0>
     False Color=$Particles.MULTIPLIED DARk
    )
   ) {enabled=false, isSetParameters=true}
   /SetVariables(
    Particles.Color=&LerpLinearColors(
     StartColor=&MultiplyLinearColor(Linear Color=$Particles.Color, Scale Factor=<1 1 1 0>)
     EndColor=$Particles.Color
     LerpFactor=&FloatFromCurve(
      FloatCurve=DataInterface() {propertyValues {Curve {defaultValue=0, keys [<at 0 0 RCIM_Constant>,<at 0.7 0 RCIM_Constant>,<at 0.71 1 RCIM_Constant>,<at 1 1 RCIM_Constant>]}}}
      CurveIndex=&DotProduct(Vector A=$Particles.SpriteFacing, Vector B=<0 0 1>)
     )
    )
    Particles.MULTIPLIED COLOR:LinearColor=&LerpLinearColors(
     StartColor=&MultiplyLinearColor(Linear Color=$Particles.MULTIPLIED COLOR, Scale Factor=<1 1 1 0>)
     EndColor=$Particles.MULTIPLIED COLOR
     LerpFactor=&FloatFromCurve(
      FloatCurve=DataInterface() {propertyValues {Curve {defaultValue=0, keys [<at 0 0 RCIM_Constant>,<at 0.7 0 RCIM_Constant>,<at 0.71 1 RCIM_Constant>,<at 1 1 RCIM_Constant>]}}}
      CurveIndex=&DotProduct(Vector A=$Particles.SpriteFacing, Vector B=<0 0 1>)
     )
    )
    Particles.MULTIPLIED DARk:LinearColor=&LerpLinearColors(
     StartColor=&MultiplyLinearColor(Linear Color=$Particles.MULTIPLIED DARk, Scale Factor=<1 1 1 0>)
     EndColor=$Particles.MULTIPLIED DARk
     LerpFactor=&FloatFromCurve(
      FloatCurve=DataInterface() {propertyValues {Curve {defaultValue=0, keys [<at 0 0 RCIM_Constant>,<at 0.7 0 RCIM_Constant>,<at 0.71 1 RCIM_Constant>,<at 1 1 RCIM_Constant>]}}}
      CurveIndex=&DotProduct(Vector A=$Particles.SpriteFacing, Vector B=<0 0 1>)
     )
    )
   ) {isSetParameters=true}
  ]
  rendererClasses [&NiagaraSpriteRendererProperties, &NiagaraDecalRendererProperties]
  renderers [
   {rendererClass=&NiagaraSpriteRendererProperties}
   {rendererClass=&NiagaraSpriteRendererProperties, rendererIndex=1}
   {rendererClass=&NiagaraSpriteRendererProperties, rendererIndex=2}
   {rendererClass=&NiagaraDecalRendererProperties, rendererIndex=3}
   {rendererClass=&NiagaraDecalRendererProperties, rendererIndex=4}
   {rendererClass=&NiagaraDecalRendererProperties, rendererIndex=5}
  ]
 }
 FE_Train {
  emitterUpdateScript [
   &EmitterState(
    Inactive Response=::NewEnumerator1
    Life Cycle Mode=::NewEnumerator1
    Loop Behavior=::NewEnumerator1
    Loop Duration Mode=::NewEnumerator1
    UseLoopDelay=false
   )
   /SetVariables(
    Emitter.OriginOffset:Vector3f=&TransformVector(Vector=<0 0 0>, Destination Space=::World)
    Emitter.PaddingPerCell=2
    Emitter.VolumeRadius=200
   ) {isSetParameters=true, module=SetVariables_#5}
   &SpawnParticlesInGrid(
    X Count:NiagaraInt32=&MakeIntFromFloat(Float=&FN_FrontEndFloatModifier(Gameplay=4, Front End=4))
    Y Count:NiagaraInt32=&MakeIntFromFloat(Float=&FN_FrontEndFloatModifier(Gameplay=4, Front End=4))
    Z Count=0
   )
  ]
  particleSpawnScript [
   &FN_SetSpawnGroup() {module=FN_SetSpawnGroup001}
   &InitializeParticle(
    Position Offset=<0 0 2>
    Position Offset Coordinate Space=::Local
    Sprite Rotation Mode=::NewEnumerator0
    Sprite Size Mode=::NewEnumerator3
    Uniform Sprite Size=55
    UsePositionOffset=true
    Write Material Random=true
   )
   &GridLocation(
    Coordinate Space:ENiagaraCoordinateSpace=::World
    Module.Jitter Inside Grid Cells=false
    Module.Override Random Seed=false
    Normalize Offsets=false
    XYZ Dimensions:Vector3f=&MakeVector(
     X=&Divide_Float(A=$Emitter.VolumeRadius, B=$Emitter.PaddingPerCell)
     Y=&Divide_Float(A=$Emitter.VolumeRadius, B=$Emitter.PaddingPerCell)
     Z=0
    )
   )
   &PartitionParticles(Normalized Index Includes One=true, Number Of Partitions=8)
   /SetVariables(Particles.LOGIC COLOR:LinearColor=&MakeCustomLinearColorFromBool(
     Bool=&SetBoolByFloatComparison(
      A=&MakeFloatFromInt(Int=$Particles.PartitionParticles.Partition)
      B=4
      Comparison Type=::NewEnumerator2
     )
     True Color=&MakeCustomLinearColorFromBool(
      Bool=&SetBoolByFloatComparison(
       A=&MakeFloatFromInt(Int=&ModuloInt(A=$Particles.PartitionParticles.Partition, B=2))
       B=0
      )
      True Color=<rgba 0.4219 0.533 1 1>
      False Color=<rgba 0.0503 0.0544 0.0938 1>
     )
     False Color=&MakeCustomLinearColorFromBool(
      Bool=&SetBoolByFloatComparison(
       A=&MakeFloatFromInt(Int=&ModuloInt(A=$Particles.PartitionParticles.Partition, B=2))
       B=0
      )
      True Color=<rgba 0.0503 0.0544 0.0938 1>
      False Color=<rgba 0.4219 0.533 1 1>
     )
    )) {isSetParameters=true}
   /SetVariables(Particles.DARKCOLOR:LinearColor=<rgba 8.593 15.6344 40 1>) {isSetParameters=true}
   /SetVariables(Particles.SpriteSize=&Multiply_Vector2D_ByFloat(Float=&FN_FrontEndFloatModifier(Gameplay=1.82, Front End=0.91), Vector2D=$Particles.SpriteSize)) {isSetParameters=true}
   /SetVariables(
    Particles.DARKCOLOR:LinearColor=&Multiply_LinearColorByFloat(Linear Color=$Particles.DARKCOLOR, Float=&FN_FrontEndFloatModifier(Gameplay=1.3, Front End=1))
    Particles.LOGIC COLOR:LinearColor=&Multiply_LinearColorByFloat(Linear Color=$Particles.LOGIC COLOR, Float=&FN_FrontEndFloatModifier(Gameplay=0.6, Front End=1))
   ) {isSetParameters=true}
  ]
  particleUpdateScript [
   &ParticleState(Kill Particles When Lifetime Has Elapsed=false, Let Infinitely Lived Particles Die When Emitter Deactivates=true)
   /SetVariables() {isSetParameters=true}
   &ScaleSpriteSize(
    Scale Sprite Size Mode=::NewEnumerator0
    Uniform Scale Factor=&Add_Float(A=$System.Rail Float, B=&FN_FrontEndFloatModifier(Gameplay=0, Front End=2))
   )
   &SocketLocation(
    Orientation Sampling=::NewEnumerator1
    Position Sampling=::NewEnumerator1
    SamplePosOffset=true
    Sampled Position Offset=&MakeVector(
     X=0
     Y=0
     Z=&MakeCustomFloatFromBool(Bool=&Boolean_OrOperation(Bool1=$User.bIsFrontEnd, Bool2=$User.bIsFrontEndPreview), True Float=0.4, False Float=5)
    )
    Socket Reader=$System.Socket Reader
    Socket Sampling Mode=::NewEnumerator3
    VelocitySpeedLimit=false
   )
   &Timeline(
    Additional Output:ENiagara_TimelineOutput=::NewEnumerator1
    Additional Output Mode:ENiagara_TimelineOutputMode=::NewEnumerator1
    Float Curve=DataInterface() {propertyValues {Curve {keys [{arriveTangent=0.9218, interpMode=RCIM_Cubic, leaveTangent=0.9218, tangentMode=RCTM_User, time=0, value=0}, <at 1 1 RCIM_Cubic>]}}}
    Play=&Boolean_NotOperation(Bool=$Particles.RegionMask.IsInsideMask)
    Play Rate=&MakeCustomFloatFromBool(
     Bool=&SetBoolByFloatComparison(A=&VectorLength(Vector=$User.OwnerLocalVelocity), B=160)
     True Float=10
     False Float=1
    )
    Rewind Rate:NiagaraFloat=&MakeCustomFloatFromBool(
     Bool=&SetBoolByFloatComparison(A=&VectorLength(Vector=$User.OwnerLocalVelocity), B=160)
     True Float=10
     False Float=3
    )
   )
   /SetVariables(Particles.PositionCenter:NiagaraPosition=&AddVectorToPosition(Position=$Engine.Owner.Position, Vector=$Emitter.OriginOffset)) {isSetParameters=true, module=SetVariables_#6}
   /SetVariables(Particles.RemappedDistance:NiagaraFloat=&NormalizeDistanceRange(
     Start Position:NiagaraPosition=&SimulationPosition()
     End Position:NiagaraPosition=$Particles.Position
     Distance:NiagaraFloat=&Multiply_Float(A=340, B=0.6)
     Invert Normalized Range:NiagaraBool=false
    )) {isSetParameters=true}
   /SetVariables(Particles.SpriteFacing=<0 0 1>) {isSetParameters=true}
   /SetVariables(Particles.NEWPOSITION:NiagaraPosition=&AddVectorToPosition(
     Position=&AddVectorToPosition(
      Position=$Particles.Position
      Vector=&Add_Vector(
       A=&Multiply_Vector(
        A=<0 0 1>
        B=&VectorFromFloat(Value=&MakeFloatFromVector(
          Vector=&VectorFromFloat(Value=&Waveform(
            Global Amplitude Scale=&Multiply_Float(
             A=&FloatFromCurve(FloatCurve=DataInterface() {propertyValues {Curve {keys [<at 0.7 0>,<at 0.9 1>]}}}, CurveIndex=$Particles.RemappedDistance)
             B=&RandomRangeFloat(Minimum=10, Maximum=15)
            )
            "Waveform [1]"=::NewEnumerator0
            "[1] Frequency"=&RandomRangeFloat(Minimum=0.1, Maximum=0.2)
            "[1] Amplitude Min/Max"=<0.1 1>
           ))
          Channel=::NewEnumerator0
         ))
       )
       B=&Lerp_Vector(
        A=&Multiply_Vector(A=<0 0 1>, B=&VectorFromFloat(Value=30))
        B=&Multiply_Vector(A=<0 0 1>, B=<2 2 2>)
        Alpha=&DotProduct(Vector A=<0 0 1>, Vector B=$Particles.SpriteFacing)
       )
      )
     )
     Vector=&Lerp_Vector(
      A=&Multiply_Vector(
       A=<0 0 1>
       B=&VectorFromFloat(Value=&FloatFromCurve(
         FloatCurve=DataInterface() {propertyValues {Curve {keys [
             {arriveTangentWeight=0.1562, leaveTangentWeight=0.2983, time=0.4, value=0}
             <at 0.5002 1>
             {arriveTangent=5, time=0.6, value=1}
             {interpMode=RCIM_Cubic, tangentMode=RCTM_Break, time=1, value=0}
            ]}}}
         CurveIndex=$Particles.RemappedDistance
         Scale Curve=5
        ))
      )
      B=&Multiply_Vector(A=<0 0 1>, B=&VectorFromFloat(Value=0))
      Alpha=&DotProduct(Vector A=<0 0 1>, Vector B=<0 0 1>)
     )
    )) {isSetParameters=true}
   /SetVariables(Particles.NEWFACING:Vector3f=&Add_Vector(
     A=<0 0 1>
     B=&MakeVector(
      X=&Waveform(
       Global Amplitude Scale=&Multiply_Float(
        A=&FloatFromCurve(FloatCurve=DataInterface() {propertyValues {Curve {keys [<at 0.41 0>,<at 0.6 1>]}}}, CurveIndex=$Particles.RemappedDistance)
        B=&RandomRangeFloat(Minimum=0.01, Maximum=0.03)
       )
       "Waveform [1]"=::NewEnumerator0
       "[1] Frequency"=&RandomRangeFloat(Minimum=0.25, Maximum=0.5)
       "[1] Amplitude Min/Max"=<-1 1>
      )
      Y=&Waveform(
       Global Amplitude Scale=&Multiply_Float(A=$Particles.RemappedDistance, B=0.02)
       "Waveform [1]"=::NewEnumerator0
       "[1] Frequency"=&RandomRangeFloat(Minimum=0.2, Maximum=0.5)
       "[1] Amplitude Min/Max"=<-1 1>
      )
      Z=0
     )
    )) {isSetParameters=true}
   /SetVariables(
    Particles.BaseDirection:Vector3f=<0 0 1>
    Particles.WaveFormDirection:Vector3f=&Multiply_Vector(A=<0 0 1>, B=<1 1 1>)
   ) {isSetParameters=true}
   &self:TreadmillParticles(
    GridSize:NiagaraFloat=&Multiply_Float(A=$Emitter.VolumeRadius, B=2)
    InputPosition:NiagaraPosition=$Particles.PositionCenter
   ) {module=TreadmillParticles}
   /SetVariables(
    Particles.Position=&AddVectorToPosition(
     Position=&MakeVector(
      X=&MakeFloatFromVector(Vector=Linked() {linkedVariable {name=Particles.Position, type=NiagaraPosition}}, Channel=::NewEnumerator0)
      Y=&MakeFloatFromVector(Vector=Linked() {linkedVariable {name=Particles.Position, type=NiagaraPosition}}, Channel=::NewEnumerator1)
      Z=&Add_Float(
       A=&MakeFloatFromVector(Vector=Linked() {linkedVariable {name=Output.SocketLocation.SampledPosition, type=NiagaraPosition}}, Channel=::NewEnumerator2)
       B=0
      )
     )
     Vector=&FN_FrontEndVector3Modifier(
      Gameplay=<0 0 0>
      Front End=&RotateVectorByQuaternion(
       Vector To Rotate=&Multiply_VectorByFloat(Vector=<0 -300 0>, Float=$Engine.DeltaTime)
       Quaternion=$Output.SocketLocation.SampledRotation
      )
     )
    )
    Particles.SpriteFacing=<0 0 1>
   ) {isSetParameters=true}
   &RegionMask(
    Invert Region=true
    Offset:Vector3f=$Emitter.OriginOffset
    Region Dimensions:Vector3f=<1000 1000 1000>
    Region Falloff=0
    Region Type:ENiagaraDebugDrawShape=::NewEnumerator1
    UseOffset=true
    UseRotation=false
   )
   &Color(
    Color=&MakeCustomLinearColorFromBool(
     Bool=&SetBoolByFloatComparison(
      A=&MakeFloatFromInt(Int=$Particles.PartitionParticles.Partition)
      B=4
      Comparison Type=::NewEnumerator2
     )
     True Color=&MakeCustomLinearColorFromBool(
      Bool=&SetBoolByFloatComparison(
       A=&MakeFloatFromInt(Int=&ModuloInt(A=$Particles.PartitionParticles.Partition, B=2))
       B=0
      )
      True Color=<rgba 0.1495 0.156 0.2031 1>
      False Color=<rgba 0.002 0.0033 0.0156 1>
     )
     False Color=&MakeCustomLinearColorFromBool(
      Bool=&SetBoolByFloatComparison(
       A=&MakeFloatFromInt(Int=&ModuloInt(A=$Particles.PartitionParticles.Partition, B=2))
       B=0
      )
      True Color=<rgba 0.002 0.0033 0.0156 1>
      False Color=<rgba 0.1471 0.1541 0.2031 1>
     )
    )
    Scale Alpha=&Multiply_Float(
     A=&Multiply_Float(
      A=&FloatFromCurve(
       FloatCurve=DataInterface() {propertyValues {Curve {keys [
           {arriveTangentWeight=0.1562, leaveTangentWeight=0.2983, time=0.4, value=0.5}
           {arriveTangentWeight=0.0898, interpMode=RCIM_Cubic, leaveTangentWeight=0.2693, tangentMode=RCTM_User, time=0.5, value=0.5}
           {interpMode=RCIM_Cubic, tangentMode=RCTM_Break, time=1, value=0}
          ]}}}
       CurveIndex=$Particles.RemappedDistance
       Scale Curve=&Add_Float(A=$System.Rail Float, B=&FN_FrontEndFloatModifier(Gameplay=0, Front End=1))
      )
      B=&FloatFromCurve(FloatCurve=DataInterface() {propertyValues {Curve {
          defaultValue=1
          keys [{interpMode=RCIM_Cubic, tangentMode=RCTM_User, time=0, value=0}, {interpMode=RCIM_Cubic, tangentMode=RCTM_User, time=1, value=1}]
         }}})
     )
     B=&FN_FrontEndFloatModifier(Gameplay=1, Front End=1)
    )
    Scale Color=&VectorFromFloat(Value=&Multiply_Float(
      A=&Multiply_Float(A=$Particles.Timeline.Float, B=&Multiply_Float(A=1, B=1))
      B=&Lerp_Float(
       A=&FN_FrontEndFloatModifier(Gameplay=0.5, Front End=1)
       B=0.1
       Alpha=$Particles.INTERPOLATE INDOORS
      )
     ))
   ) {module=Color002}
   /SetVariables(
    Particles.GlowColor=&Multiply_LinearColorByFloat(Linear Color=$Particles.Color, Float=1)
    Particles.GlowSize=&Multiply_Vector2D_ByFloat(Float=1.6, Vector2D=$Particles.SpriteSize)
   ) {isSetParameters=true, module=SetVariables_#7}
   /SetVariables(Particles.MULTIPLIED COLOR:LinearColor=&MultiplyLinearColor(
     Linear Color=$Particles.LOGIC COLOR
     Scale Factor=&MakeVector4(
      X=&FloatFromCurve(
       FloatCurve=DataInterface() {propertyValues {Curve {
          defaultValue=1
          keys [{interpMode=RCIM_Cubic, tangentMode=RCTM_User, time=0, value=0}, {interpMode=RCIM_Cubic, tangentMode=RCTM_User, time=1, value=1}]
         }}}
       Scale Curve=&Lerp_Float(A=1, B=0.5, Alpha=$Particles.INTERPOLATE INDOORS)
      )
      Y=&FloatFromCurve(
       FloatCurve=DataInterface() {propertyValues {Curve {
          defaultValue=1
          keys [{interpMode=RCIM_Cubic, tangentMode=RCTM_User, time=0, value=0}, {interpMode=RCIM_Cubic, tangentMode=RCTM_User, time=1, value=1}]
         }}}
       Scale Curve=&Lerp_Float(A=1, B=0.5, Alpha=$Particles.INTERPOLATE INDOORS)
      )
      Z=&FloatFromCurve(
       FloatCurve=DataInterface() {propertyValues {Curve {
          defaultValue=1
          keys [{interpMode=RCIM_Cubic, tangentMode=RCTM_User, time=0, value=0}, {interpMode=RCIM_Cubic, tangentMode=RCTM_User, time=1, value=1}]
         }}}
       Scale Curve=&Lerp_Float(A=1, B=0.5, Alpha=$Particles.INTERPOLATE INDOORS)
      )
      W=&Multiply_Float(
       A=&FloatFromCurve(
        FloatCurve=DataInterface() {propertyValues {Curve {keys [
            {interpMode=RCIM_Cubic, tangentMode=RCTM_Break, time=0, value=0.5}
            {interpMode=RCIM_Cubic, tangentMode=RCTM_Break, time=0.1331, value=0.3}
            {interpMode=RCIM_Cubic, tangentMode=RCTM_Break, time=0.2358, value=0.3}
            {interpMode=RCIM_Cubic, tangentMode=RCTM_Break, time=0.5, value=0}
           ]}}}
        CurveIndex=&NormalizeDistanceRange(
         Start Position:NiagaraPosition=&SimulationPosition()
         End Position:NiagaraPosition=$Particles.Position
         Distance:NiagaraFloat=&Multiply_Float(A=340, B=1.29)
         Invert Normalized Range:NiagaraBool=false
        )
        Scale Curve=&Add_Float(A=$System.Rail Float, B=&FN_FrontEndFloatModifier(Gameplay=0, Front End=1))
       )
       B=&FloatFromCurve(FloatCurve=DataInterface() {propertyValues {Curve {
           defaultValue=1
           keys [{interpMode=RCIM_Cubic, tangentMode=RCTM_User, time=0, value=0}, {interpMode=RCIM_Cubic, tangentMode=RCTM_User, time=1, value=1}]
          }}})
      )
     )
    )) {isSetParameters=true}
   /SetVariables(Particles.MULTIPLIED DARk:LinearColor=&MultiplyLinearColor(
     Linear Color=$Particles.DARKCOLOR
     Scale Factor=&MakeVector4(
      X=&FloatFromCurve(
       FloatCurve=DataInterface() {propertyValues {Curve {
          defaultValue=1
          keys [{interpMode=RCIM_Cubic, tangentMode=RCTM_User, time=0, value=0}, {interpMode=RCIM_Cubic, tangentMode=RCTM_User, time=1, value=1}]
         }}}
       Scale Curve=&Lerp_Float(A=1, B=0.18, Alpha=$Particles.INTERPOLATE INDOORS)
      )
      Y=&FloatFromCurve(
       FloatCurve=DataInterface() {propertyValues {Curve {
          defaultValue=1
          keys [{interpMode=RCIM_Cubic, tangentMode=RCTM_User, time=0, value=0}, {interpMode=RCIM_Cubic, tangentMode=RCTM_User, time=1, value=1}]
         }}}
       Scale Curve=&Lerp_Float(A=1, B=0.18, Alpha=$Particles.INTERPOLATE INDOORS)
      )
      Z=&FloatFromCurve(
       FloatCurve=DataInterface() {propertyValues {Curve {
          defaultValue=1
          keys [{interpMode=RCIM_Cubic, tangentMode=RCTM_User, time=0, value=0}, {interpMode=RCIM_Cubic, tangentMode=RCTM_User, time=1, value=1}]
         }}}
       Scale Curve=&Lerp_Float(A=1, B=0.18, Alpha=$Particles.INTERPOLATE INDOORS)
      )
      W=&Multiply_Float(
       A=&FloatFromCurve(
        FloatCurve=DataInterface() {propertyValues {Curve {keys [
            {arriveTangentWeight=0.0898, interpMode=RCIM_Cubic, leaveTangentWeight=0.2693, tangentMode=RCTM_User, time=0.42, value=1}
            {interpMode=RCIM_Cubic, tangentMode=RCTM_Break, time=1, value=0}
           ]}}}
        CurveIndex=$Particles.RemappedDistance
        Scale Curve=&Add_Float(A=$System.Rail Float, B=&FN_FrontEndFloatModifier(Gameplay=-0.5, Front End=1))
       )
       B=&FloatFromCurve(FloatCurve=DataInterface() {propertyValues {Curve {
           defaultValue=1
           keys [{interpMode=RCIM_Cubic, tangentMode=RCTM_User, time=0, value=0}, {interpMode=RCIM_Cubic, tangentMode=RCTM_User, time=1, value=1}]
          }}})
      )
     )
    )) {isSetParameters=true}
   &SolveForcesAndVelocity(Clamp Velocity=false, Limit Acceleration=false)
  ]
  rendererClasses [&NiagaraSpriteRendererProperties]
  renderers [
   {rendererClass=&NiagaraSpriteRendererProperties}
   {rendererClass=&NiagaraSpriteRendererProperties, rendererIndex=1}
   {rendererClass=&NiagaraSpriteRendererProperties, rendererIndex=2}
  ]
 }
 Minimal {
  emitterUpdateScript [
   &EmitterState(Life Cycle Mode=::NewEnumerator1, Loop Behavior=::NewEnumerator1, Loop Duration=1, UseLoopDelay=false)
   &SpawnBurst_Instantaneous(Use Spawn Probability=false, UseLoopCountLimit=false)
  ]
  particleSpawnScript [
   &InitializeParticle(
    Color=&LerpLinearColors(
     StartColor=<rgba 0.495 0.75 2.5521 0.17>
     EndColor=<rgba 0.6799 1.4424 3 0.2>
     LerpFactor=&FN_FrontEndFloatModifier(Gameplay=0, Front End=1)
    )
    Mesh Scale=<0.096 0.095 0.1>
    Mesh Scale Mode=::NewEnumerator4
    Position Offset=&FN_FrontEndVector3Modifier(Gameplay=<-7 2 -6>, Front End=<-7 2 -10>)
    UsePositionOffset=true
    Write Material Random=true
   )
   /SetVariables(Particles.Linear Color=&MultiplyLinearColor(Linear Color=$Particles.Color, Scale Factor=<0.8 0.8 1 0.8>)) {isSetParameters=true}
   &CameraOffset(Camera Offset Amount=&FN_FrontEndFloatModifier(Gameplay=0, Front End=5))
  ]
  particleUpdateScript [
   &ParticleState(Kill Particles When Lifetime Has Elapsed=false)
   &SkeletalMeshLocation(
    Bone Sampling Mode=::NewEnumerator4
    MaskSpawnGroup=false
    Sampled Position Offset=&FN_FrontEndVector3Modifier(Gameplay=<0 2 0>, Front End=<0 2 -0.5>)
    Sampled Velocity Coordinate Space:ENiagaraCoordinateSpace=::World
    Skeletal Mesh=DataInterface() {propertyValues {
      DefaultMesh=null
      ExcludeBoneName=root
      OptionalComponentClass=null
      PreviewMesh=null
      SkinningMode=SkinOnTheFly
      SoftSourceActor=null
      SourceMode=Default
      UvSetIndex=0
      WholeMeshLOD=-1
      bExcludeBone=false
      bReadDeformedGeometry=true
      bRequireCurrentFrameData=true
      ComponentTags []
      FilteredBones [root]
      FilteredSockets []
      MeshUserParameter {parameter {name=None, typeDefHandle {registeredTypeIndex=1}, varData []}}
      SamplingRegions []
     }}
   )
   /SetVariables(Particles.Debug:NiagaraFloat=&RemapRange(
     Input Value To Remap=&VectorLength(Vector=$Particles.SkeletalMeshLocation.SampledVelocity)
     Input Min=1
     Input Max=250
     Output Min=0
     Output Max=1
    )) {isSetParameters=true}
   &ScaleColor(
    Scale Alpha=&FN_FrontEndFloatModifier(Gameplay=$Particles.Debug, Front End=1)
    ScaleA=true
    ScaleRGB=true
   )
   /SetVariables(Particles.Linear Color=&MultiplyLinearColor(
     Linear Color=$Particles.Initial.Linear Color
     Scale Factor=&MakeVector4(X=1, Y=1, Z=1, W=$Particles.Debug)
    )) {isSetParameters=true}
  ]
  rendererClasses [&NiagaraMeshRendererProperties]
  renderers [
   {rendererClass=&NiagaraMeshRendererProperties, rendererIndex=0}
   {rendererClass=&NiagaraMeshRendererProperties, rendererIndex=1}
  ]
 }
 Minimal001 {
  emitterUpdateScript [
   &EmitterState(Life Cycle Mode=::NewEnumerator1, Loop Duration=1, UseLoopDelay=false)
   &SpawnRate(SpawnRate=8, Use Spawn Probability=false)
   &SpawnBurst_Instantaneous(Use Spawn Probability=false, UseLoopCountLimit=false)
  ]
  particleSpawnScript [
   &InitializeParticle(
    Color=&LerpLinearColors(
     StartColor=<rgba 0.495 0.75 2.5521 0.05>
     EndColor=<rgba 0.1771 0.4031 2 0.3>
     LerpFactor=&FN_FrontEndFloatModifier(Gameplay=0, Front End=1)
    )
    Lifetime=0.15
    Mesh Scale=<7 5 0.5>
    Mesh Scale Mode=::NewEnumerator4
    Position Offset=<-7 0 -1.5>
    Sprite Size Mode=::NewEnumerator3
    Uniform Sprite Size=1
    UsePositionOffset=true
    Write Material Random=true
   )
   &InitialMeshOrientation(
    Mesh Orientation Mode=::NewEnumerator1
    Rotation=&RandomRangeVector(Minimum=<0 0 0>, Maximum=<0 0 1>)
    Use Rotation Vector=true
   )
   &CameraOffset(Camera Offset Amount=&FN_FrontEndFloatModifier(Gameplay=5, Front End=5))
  ]
  particleUpdateScript [
   &ParticleState()
   &SkeletalMeshLocation(
    Bone Sampling Mode=::NewEnumerator4
    MaskSpawnGroup=false
    Position Sampling=::NewEnumerator2
    Sampled Velocity Coordinate Space:ENiagaraCoordinateSpace=::World
    Skeletal Mesh=DataInterface() {propertyValues {
      DefaultMesh=null
      ExcludeBoneName=root
      OptionalComponentClass=null
      PreviewMesh=null
      SkinningMode=SkinOnTheFly
      SoftSourceActor=null
      SourceMode=Default
      UvSetIndex=0
      WholeMeshLOD=-1
      bExcludeBone=false
      bReadDeformedGeometry=true
      bRequireCurrentFrameData=true
      ComponentTags []
      FilteredBones [root]
      FilteredSockets []
      MeshUserParameter {parameter {name=None, typeDefHandle {registeredTypeIndex=1}, varData []}}
      SamplingRegions []
     }}
   )
   &DynamicMaterialParameters(
    Index 1 Param 2=-0.25
    Module.Index0Float0Write=false
    Module.Index0Float1Write=false
    Module.Index0Float2Write=true
    Module.Index0Float3Write=true
    Module.Index1Float0Write=true
    Module.Index1Float1Write=true
    Module.Index1Float2Write=true
    Module.Index1Float3Write=true
    Write Parameter Index 1=::NewEnumerator1
   )
   /SetVariables(Particles.Debug:NiagaraFloat=&RemapRange(
     Input Value To Remap=&VectorLength(Vector=$Particles.SkeletalMeshLocation.SampledVelocity)
     Input Min=1
     Input Max=250
     Output Min=0
     Output Max=1
    )) {isSetParameters=true}
   &ScaleMeshSize(Scale Factor=&Multiply_VectorByFloat(
     Vector=<1 1 1>
     Float=&FloatFromCurve(FloatCurve=DataInterface() {propertyValues {Curve {keys [<at 0 1>,<at 1 1.01>]}}})
    ))
   &ScaleColor(
    Scale Alpha=&FN_FrontEndFloatModifier(Gameplay=$Particles.Debug, Front End=1)
    ScaleA=true
    ScaleRGB=true
   )
  ]
  rendererClasses [&NiagaraMeshRendererProperties]
  renderers [
   {rendererClass=&NiagaraMeshRendererProperties, rendererIndex=0}
   {rendererClass=&NiagaraMeshRendererProperties, rendererIndex=1}
  ]
 }
 Minimal002 {
  emitterUpdateScript [
   &EmitterState(
    Inactive Response=::NewEnumerator1
    Life Cycle Mode=::NewEnumerator1
    Loop Behavior=::NewEnumerator1
    Loop Delay=0.3
    Loop Duration Mode=::NewEnumerator1
    UseLoopDelay=true
   )
   &SpawnBurst_Instantaneous(Use Spawn Probability=false, UseLoopCountLimit=false)
  ]
  particleSpawnScript [
   &InitializeParticle(
    Color=<rgba 0.0014 0.1452 1 0.6>
    Lifetime=0.5
    Position Offset=<0 0 2>
    Sprite Size Mode=::NewEnumerator3
    Uniform Sprite Size=250
    UsePositionOffset=true
    Write Material Random=true
   )
   &SkeletalMeshLocation(
    Bone Sampling Mode=::NewEnumerator4
    MaskSpawnGroup=false
    Orientation Sampling=::NewEnumerator1
    Position Sampling=::NewEnumerator2
    Skeletal Mesh=DataInterface() {propertyValues {
      DefaultMesh=null
      ExcludeBoneName=root
      OptionalComponentClass=null
      PreviewMesh=null
      SkinningMode=SkinOnTheFly
      SoftSourceActor=null
      SourceMode=Default
      UvSetIndex=0
      WholeMeshLOD=-1
      bExcludeBone=false
      bReadDeformedGeometry=true
      bRequireCurrentFrameData=true
      ComponentTags []
      FilteredBones [root]
      FilteredSockets []
      MeshUserParameter {parameter {name=None, typeDefHandle {registeredTypeIndex=1}, varData []}}
      SamplingRegions []
     }}
   ) {module=SkeletalMeshLocation001}
   /SetVariables(Particles.Ring Color:LinearColor=<rgba 0.384 0.672 3 0.3>, Particles.Ring Size:Vector2f=<95 95>) {isSetParameters=true}
   /SetVariables(Particles.GlowSize=<150 150>) {isSetParameters=true}
   &CameraOffset(Camera Offset Amount=5)
  ]
  particleUpdateScript [
   &ParticleState(Kill Particles When Lifetime Has Elapsed=false)
   &SkeletalMeshLocation(
    Bone Sampling Mode=::NewEnumerator4
    MaskSpawnGroup=false
    Sampled Velocity Coordinate Space:ENiagaraCoordinateSpace=::World
    Skeletal Mesh=DataInterface() {propertyValues {
      DefaultMesh=null
      ExcludeBoneName=root
      OptionalComponentClass=null
      PreviewMesh=null
      SkinningMode=SkinOnTheFly
      SoftSourceActor=null
      SourceMode=Default
      UvSetIndex=0
      WholeMeshLOD=-1
      bExcludeBone=false
      bReadDeformedGeometry=true
      bRequireCurrentFrameData=true
      ComponentTags []
      FilteredBones [root]
      FilteredSockets []
      MeshUserParameter {parameter {name=None, typeDefHandle {registeredTypeIndex=1}, varData []}}
      SamplingRegions []
     }}
   )
   /SetVariables(Particles.Debug:NiagaraFloat=&RemapRange(
     Input Value To Remap=&VectorLength(Vector=$Particles.SkeletalMeshLocation.SampledVelocity)
     Input Min=1
     Input Max=50
     Output Min=0
     Output Max=1
    )) {isSetParameters=true}
   /SetVariables(Particles.NEWPOSITION:NiagaraPosition=&AddVectorToPosition(Position=$Particles.Position, Vector=&Lerp_Vector(A=<0 0 1>, B=<0 0 2>, Alpha=$Particles.Debug))) {isSetParameters=true}
   &SpriteFacingAndAlignment(
    Sprite Facing=&ReturnMeshOrientationAxis(Mesh Orientation Axis=::ZAxis, Mesh Orientation Quat=$Particles.SkeletalMeshLocation.SampledRotation)
    WriteAlignment=false
    WriteFacing=true
   )
   &ScaleColor(
    Scale Alpha=&FloatFromCurve(FloatCurve=DataInterface() {propertyValues {Curve {
        defaultValue=0
        keys [{interpMode=RCIM_Cubic, tangentMode=RCTM_User, time=0, value=0}, {interpMode=RCIM_Cubic, tangentMode=RCTM_User, time=1, value=1}]
       }}})
    ScaleA=true
    ScaleRGB=true
   ) {module=ScaleColor001}
   /SetVariables(Particles.Ring Color:LinearColor=&MultiplyLinearColor(
     Linear Color=$Particles.Initial.Ring Color
     Scale Factor=&MakeVector4(
      X=1
      Y=1
      Z=1
      W=&FloatFromCurve(FloatCurve=DataInterface() {propertyValues {Curve {
          defaultValue=0
          keys [{interpMode=RCIM_Cubic, tangentMode=RCTM_User, time=0, value=0}, {interpMode=RCIM_Cubic, tangentMode=RCTM_User, time=1, value=1}]
         }}})
     )
    )) {isSetParameters=true}
  ]
  rendererClasses [&NiagaraSpriteRendererProperties]
  renderers [
   {rendererClass=&NiagaraSpriteRendererProperties}
   {rendererClass=&NiagaraSpriteRendererProperties, rendererIndex=1}
   {rendererClass=&NiagaraSpriteRendererProperties, rendererIndex=2}
  ]
 }
 Minimal006 {
  emitterUpdateScript [
   &EmitterState(
    Inactive Response=::NewEnumerator1
    Life Cycle Mode=::NewEnumerator1
    Loop Behavior=::NewEnumerator1
    Loop Delay=0.3
    Loop Duration Mode=::NewEnumerator1
    UseLoopDelay=true
   )
   &SpawnRate(SpawnRate=20, Use Spawn Probability=false)
  ]
  particleSpawnScript [
   &SkeletalMeshLocation(
    Bone Sampling Mode=::NewEnumerator4
    MaskSpawnGroup=false
    Orientation Sampling=::NewEnumerator1
    Position Sampling=::NewEnumerator2
    Skeletal Mesh=DataInterface() {propertyValues {
      DefaultMesh=null
      ExcludeBoneName=root
      OptionalComponentClass=null
      PreviewMesh=null
      SkinningMode=SkinOnTheFly
      SoftSourceActor=null
      SourceMode=Default
      UvSetIndex=0
      WholeMeshLOD=-1
      bExcludeBone=false
      bReadDeformedGeometry=true
      bRequireCurrentFrameData=true
      ComponentTags []
      FilteredBones [root]
      FilteredSockets []
      MeshUserParameter {parameter {name=None, typeDefHandle {registeredTypeIndex=1}, varData []}}
      SamplingRegions []
     }}
   ) {module=SkeletalMeshLocation001}
   /SetVariables(Particles.Debug:NiagaraFloat=&RemapRange(
     Input Value To Remap=&VectorLength(Vector=$Particles.SkeletalMeshLocation001.SampledVelocity)
     Input Min=1
     Input Max=250
     Output Min=0
     Output Max=1
    )) {isSetParameters=true}
   &InitializeParticle(
    Color=&MultiplyLinearColor(
     Linear Color=<rgba 0.4237 0.5736 2 1>
     Scale Factor=&MakeVector4(
      X=1
      Y=1
      Z=1
      W=&FloatFromCurve(FloatCurve=DataInterface() {propertyValues {Curve {keys [<at 0 1>,<at 1 0>]}}}, CurveIndex=$Particles.Debug)
     )
    )
    Lifetime=&RandomRangeFloat(Minimum=0.3, Maximum=2)
    Sprite Size Mode=::NewEnumerator3
    Uniform Sprite Size=&RandomRangeFloat(Minimum=3.5, Maximum=9.3)
    UsePositionOffset=false
    Write Material Random=true
   )
   &AddVelocity(
    Rotation Coordinate Space=::Simulation
    Velocity=&ReturnMeshOrientationAxis(Mesh Orientation Axis=::ZAxis, Mesh Orientation Quat=$Particles.SkeletalMeshLocation001.SampledRotation)
    Velocity Speed Scale=50
   )
   &ShapeLocation(
    MaskSpawnGroup=false
    Offset Coordinate Space=::Local
    Offset Mode=::NewEnumerator0
    Override Seed=false
    Ring Radius=40
    Shape Origin=$Particles.SkeletalMeshLocation001.SampledPosition
    Shape Primitive=::NewEnumerator4
    UseOwnerScale=true
   )
  ]
  particleUpdateScript [
   &ParticleState()
   &ScaleSpriteSize(
    Uniform Curve Scale=2
    Uniform Curve Sprite Scale=DataInterface() {propertyValues {Curve {keys [
        {tangentMode=RCTM_User, time=0, value=1}
        <at 0.3255 0.4706>
        <at 0.3775 1.3088>
        <at 0.4253 0.7676>
        <at 0.4687 1.0735>
        <at 0.6207 0.6382>
        <at 0.6684 1.3088>
        {tangentMode=RCTM_User, time=1, value=1}
       ]}}}
   )
   &CurlNoiseForce(Bias Noise=false, Pan Noise=false)
   &SolveForcesAndVelocity(Clamp Velocity=false, Limit Acceleration=false)
   &SkeletalMeshLocation(
    Bone Sampling Mode=::NewEnumerator4
    MaskSpawnGroup=false
    Orientation Sampling=::NewEnumerator1
    Position Sampling=::NewEnumerator2
    Skeletal Mesh=DataInterface() {propertyValues {
      DefaultMesh=null
      ExcludeBoneName=root
      OptionalComponentClass=null
      PreviewMesh=null
      SkinningMode=SkinOnTheFly
      SoftSourceActor=null
      SourceMode=Default
      UvSetIndex=0
      WholeMeshLOD=-1
      bExcludeBone=false
      bReadDeformedGeometry=true
      bRequireCurrentFrameData=true
      ComponentTags []
      FilteredBones [root]
      FilteredSockets []
      MeshUserParameter {parameter {name=None, typeDefHandle {registeredTypeIndex=1}, varData []}}
      SamplingRegions []
     }}
   )
   /SetVariables(Particles.Debug:NiagaraFloat=&RemapRange(
     Input Value To Remap=&VectorLength(Vector=$Particles.SkeletalMeshLocation.SampledVelocity)
     Input Min=1
     Input Max=250
     Output Min=0
     Output Max=1
    )) {isSetParameters=true}
   &ScaleColor(
    Scale Alpha=&FloatFromCurve(
     FloatCurve=DataInterface() {propertyValues {Curve {keys [<at 0 1>,<at 1 0>]}}}
     Scale Curve=&FloatFromCurve(FloatCurve=DataInterface() {propertyValues {Curve {keys [<at 0 1>,<at 1 1>]}}}, CurveIndex=$Particles.Debug)
    )
    ScaleA=true
    ScaleRGB=true
   )
   &ScaleColor(
    Scale Alpha=&RemapRange(
     Input Value To Remap=&DistanceBetweenPositions(A=$Particles.Position, B=$Particles.SkeletalMeshLocation.SampledPosition)
     Input Min=0
     Input Max=100
     Output Min=1
     Output Max=0
    )
    ScaleA=true
    ScaleRGB=true
   ) {module=ScaleColor001}
  ]
  rendererClasses [&NiagaraSpriteRendererProperties]
  renderers [{rendererClass=&NiagaraSpriteRendererProperties}]
 }
 Minimal007 {
  emitterUpdateScript [
   &EmitterState()
   &SpawnRate(SpawnRate=150, Use Spawn Probability=false)
  ]
  particleSpawnScript [
   &PartitionParticles()
   &InitializeParticle(
    Color=&RandomRangeLinearColor(Minimum=<rgba 0.1583 0.389 1 0.5>, Maximum=<rgba 0.3232 0.6147 1 2>)
    Lifetime=&RandomRangeFloat(Minimum=0.2, Maximum=0.25)
    Position Offset=&MergeHelper_VectorSwitch(
     0=&FN_FrontEndVector3Modifier(Gameplay=<35 0 0>, Front End=<35 0 -3>)
     1=&FN_FrontEndVector3Modifier(Gameplay=<-49 0 0>, Front End=<-50 0 -3>)
     2=?
     3=?
     Index=$Particles.PartitionParticles.Partition
     Num=::NewEnumerator0
    )
    Position Offset Coordinate Space=::Local
    Sprite Size Max=<10 200>
    Sprite Size Min=<10 200>
    Sprite Size Mode=::NewEnumerator2
    UsePositionOffset=true
    Write Material Random=true
   )
   &AddVelocity(
    Cone Angle=10
    Cone Axis=<0 -1 0.2>
    Module.UseConeFalloff=false
    Module.UseVelDistribution=false
    Override Seed=false
    Velocity Mode=::NewEnumerator2
    Velocity Speed=300
   )
   &DynamicMaterialParameters(
    Index 0 Param 2=1
    Index 1 Param 2=-0.3
    Module.Index0Float0Write=true
    Module.Index0Float1Write=true
    Module.Index0Float2Write=true
    Module.Index0Float3Write=true
    Module.Index1Float0Write=true
    Module.Index1Float1Write=true
    Module.Index1Float2Write=true
    Module.Index1Float3Write=true
    Write Parameter Index 1=::NewEnumerator1
   )
  ]
  particleUpdateScript [
   &ParticleState()
   &ScaleSpriteSize(
    Non-Uniform Curve Sprite Scale=DataInterface() {propertyValues {
      XCurve {keys [{tangentMode=RCTM_User, time=0, value=1}, {tangentMode=RCTM_User, time=1, value=1}]}
      YCurve {
       defaultValue=0
       keys [{interpMode=RCIM_Cubic, tangentMode=RCTM_User, time=0, value=1}, {interpMode=RCIM_Cubic, tangentMode=RCTM_User, time=1, value=0.5}]
      }
     }}
    Scale Sprite Size Mode=::NewEnumerator3
   )
   &SolveForcesAndVelocity(Clamp Velocity=false, Limit Acceleration=false)
   &SkeletalMeshLocation(
    Bone Sampling Mode=::NewEnumerator4
    MaskSpawnGroup=false
    Orientation Sampling=::NewEnumerator1
    Position Sampling=::NewEnumerator2
    Sampled Velocity Coordinate Space:ENiagaraCoordinateSpace=::World
    Skeletal Mesh=DataInterface() {propertyValues {
      DefaultMesh=null
      ExcludeBoneName=root
      OptionalComponentClass=null
      PreviewMesh=null
      SkinningMode=SkinOnTheFly
      SoftSourceActor=null
      SourceMode=Default
      UvSetIndex=0
      WholeMeshLOD=-1
      bExcludeBone=false
      bReadDeformedGeometry=true
      bRequireCurrentFrameData=true
      ComponentTags []
      FilteredBones [root]
      FilteredSockets []
      MeshUserParameter {parameter {name=None, typeDefHandle {registeredTypeIndex=1}, varData []}}
      SamplingRegions []
     }}
   )
   /SetVariables(Particles.Debug:NiagaraFloat=&RemapRange(
     Input Value To Remap=&VectorLength(Vector=$Particles.SkeletalMeshLocation.SampledVelocity)
     Input Min=1
     Input Max=250
     Output Min=0
     Output Max=1
    )) {isSetParameters=true}
   &ScaleColor(
    Scale Alpha=&FN_FrontEndFloatModifier(Gameplay=$Particles.Debug, Front End=1)
    ScaleA=true
    ScaleRGB=true
   )
   &BlendSpace() {module=BlendSpace001}
  ]
  rendererClasses [&NiagaraSpriteRendererProperties]
  renderers [{rendererClass=&NiagaraSpriteRendererProperties}]
 }
}
summary {
 systemName=NS_Emote_MatteDecoy_Board2
 emitters [
  {emitterName=FE_Train, rendererClasses [&NiagaraSpriteRendererProperties]}
  {emitterName=Chessboard, rendererClasses [&NiagaraSpriteRendererProperties, &NiagaraDecalRendererProperties]}
  {emitterName=Minimal001, rendererClasses [&NiagaraMeshRendererProperties]}
  {emitterName=Minimal002, rendererClasses [&NiagaraSpriteRendererProperties]}
  {emitterName=Minimal006, rendererClasses [&NiagaraSpriteRendererProperties]}
  {emitterName=Minimal, rendererClasses [&NiagaraMeshRendererProperties]}
  {emitterName=Minimal007, rendererClasses [&NiagaraSpriteRendererProperties]}
 ]
 userVariables [
  User.bIsFrontEnd {type=NiagaraBool}
  User.bIsFrontEndPreview {type=NiagaraBool}
  User.BurstIndex {type=NiagaraFloat, defaultValue=0, description=""}
  User.Distance {type=NiagaraFloat, defaultValue=1, description=""}
  User.OwnerLocalVelocity {type=Vector3f, description="", defaultValue {struct=Vector3f, value=<1 1 1>}}
 ]
}
