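# Niagara system for the spawn effect of the MatteDecoy main player emote.
# It combines three emitters named for airglow, smoke and sparkles; their LWE suffix suggests lightweight emitters, with no renderers or emitter modules listed.
# SystemState is the only listed module, and user parameters expose a front-end preview flag, owner local velocity and a float SpawnCurve.
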NiagaraSystem /BRCosmetics/Animation/Game/MainPlayer/Emotes/MatteDecoy/FX/NS_Emote_MatteDecoy_Spawn
dependencies {usedDataInterfaces [], usedDynamicInputs [], usedModules [&SystemState]}
emitters {
 AirglowLWE001 {rendererClasses [], renderers []}
 SmokeLWE {rendererClasses [], renderers []}
 SparklesLWE001 {rendererClasses [], renderers []}
}
summary {
 systemName=NS_Emote_MatteDecoy_Spawn
 emitters [
  {emitterName=SparklesLWE001, rendererClasses []}
  {emitterName=AirglowLWE001, rendererClasses []}
  {emitterName=SmokeLWE, rendererClasses []}
 ]
 userVariables [
  User.bIsFrontEndPreview {type=NiagaraBool}
  User.OwnerLocalVelocity {type=Vector3f, description="", defaultValue {struct=Vector3f, value=<0 0 0>}}
  User.SpawnCurve {type=NiagaraFloat, defaultValue=1, description=""}
 ]
}
